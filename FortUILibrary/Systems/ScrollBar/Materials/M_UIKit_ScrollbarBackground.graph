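# Additive UI material for a scrollbar background in UIKit, drawing a radial shape with a body color inside a frame color.
# Two exponential radial gradients share a radius, the inner one shrunk by an offset, each with its own sharpness.
# Their product blends frame RGB into body RGB.
# Alpha lerps from body alpha to outer-masked frame alpha by the inner gradient, then is masked by the outer gradient.
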Material /FortUILibrary/Systems/ScrollBar/Materials/M_UIKit_ScrollbarBackground
graph {
 properties {bUsedWithStaticMesh=false, blendMode=BLEND_Additive, materialDomain=MD_UI, pixelDepthOffsetMode=PDOM_Legacy, usedShadingModels=Default Lit}
 nodes [
  n1 = VectorParameter() {parameterName=BodyColor, defaultValue=<rgba 1 1 1 1>}
  n2 = MaterialFunctionCall(
   UVs=TextureCoordinate()
   Radius=ScalarParameter() {defaultValue=0.5, group=CirlceTreatment, parameterName=Radius}
   Density=ScalarParameter() {defaultValue=10, parameterName=OutlineSharpness}
  ) {materialFunction=&RadialGradientExponential}
  n3 = VectorParameter() {parameterName=FrameColor, defaultValue=<rgba 1 0 0.0263 1>}
  n4 = MaterialFunctionCall(
   UVs=TextureCoordinate()
   Radius=Subtract(
    A=ScalarParameter() {defaultValue=0.5, group=CirlceTreatment, parameterName=Radius}
    B=ScalarParameter() {defaultValue=0.1, parameterName=OutlineOfset}
   )
   Density=ScalarParameter() {defaultValue=10, parameterName=BorderSharpness}
  ) {materialFunction=&RadialGradientExponential}
  Multiply(
   A=@n2.RadialGradientExponential
   B=LinearInterpolate(A=Multiply(A=@n1.A), B=Multiply(A=@n2.RadialGradientExponential, B=@n3.A), Alpha=@n4.RadialGradientExponential) {constA=1}
  )
  Reroute(None=LinearInterpolate(A=@n3.RGB, B=@n1.RGB, Alpha=Multiply(A=@n2.RadialGradientExponential, B=@n4.RadialGradientExponential)))
 ]
}
